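# Sound cue for the looping ambient sound of a device case in the Season 19 Battle Pass content, routed to the quests sound class.
# A looping wave player feeds an Enveloper whose volume curve ramps from silent to full at the start.
# It plays at reduced volume and references a shared supply-drop ambient attenuation asset for distance falloff.
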SoundCue /BattlePassS19/Audio/SFX/Device_Case_Ambient_Loop_Cue
root=Enveloper_0
nodes [
 {
  id=Enveloper_0
  children [WavePlayer_2]
  properties {pitchCurve {editorCurveData {keys [{time=0}]}}, volumeCurve {editorCurveData {keys [<at 0 0>,<at 1 1>]}}}
 }
 {id=WavePlayer_2, properties {bLooping=true, soundWaveAssetPtr=&/Device_Case_Loop}}
]
properties {
 attenuationSettings=/Game/Athena/Sounds/SupplyDrops/Fort_HiddenStar_Ambient_Loop_Att
 maxDistance=2500
 soundClassObject=&Fort_Quests
 volumeMultiplier=0.5
 attenuationOverrides {falloffDistance=700, attenuationShapeExtents {x=100}}
}
