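# Sound cue for the Device Case collect sound effect in the Season 19 Battle Pass content.
# A single wave player feeds the output directly, with a boosted volume and a dedicated attenuation asset for distance falloff.
# It routes through the non-critical in-game UI sound class.
# A concurrency override allows only one instance at a time and blocks new plays while one is active.
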SoundCue /BattlePassS19/Audio/SFX/Device_Case_Collect_Cue
root=WavePlayer_1
nodes [{id=WavePlayer_1, properties {soundWaveAssetPtr=&/Device_Case_Collect_01}}]
properties {
 attenuationSettings=&/Device_Case_Collect_Att
 bOverrideConcurrency=true
 maxDistance=3500
 soundClassObject=&Fort_UI_InGame_NonCritical
 volumeMultiplier=1.65
 concurrencyOverrides {maxCount=1, resolutionRule=PreventNew}
}
