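# Sound cue that plays the jammer device destruction effect in BattlePassS19, routed to the non-critical in-game UI sound class.
# A single wave player feeds a Modulator that holds volume fixed and leaves pitch at its default slight random variation.
# The cue boosts overall volume and sets no attenuation.
# A concurrency override allows one instance at a time, stopping the oldest when it retriggers.
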SoundCue /BattlePassS19/Audio/SFX/Device_Jammer_Destroy_Cue
root=Modulator_0
nodes [
 {id=Modulator_0, children [WavePlayer_1], properties {volumeMax=1, volumeMin=1}}
 {id=WavePlayer_1, properties {soundWaveAssetPtr=&/Device_Jammer_Destroy_01}}
]
properties {
 bOverrideConcurrency=true
 maxDistance=2097152
 soundClassObject=&Fort_UI_InGame_NonCritical
 volumeMultiplier=1.65
 concurrencyOverrides {maxCount=1, resolutionRule=StopOldest}
}
