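# Sound cue for the Season 19 Battle Pass rock dead drop device sound effect, routed to the Fort_Quests sound class.
# A Random node picks one of three RockDeadDrop wave variations with equal weight and feeds a Modulator at the root, left at its default settings.
# A dedicated dead drop attenuation asset handles spatial falloff.
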SoundCue /BattlePassS19/Audio/SFX/Device_S19_RockDeadDrop_Cue
root=Modulator_0
nodes [
 {id=Modulator_0, children [Random_0]}
 {id=Random_0, children [WavePlayer_0,WavePlayer_1,WavePlayer_2], properties {weights [1,1,1]}}
 {id=WavePlayer_0, properties {soundWaveAssetPtr=&/Device_S19_RockDeadDrop_03}}
 {id=WavePlayer_1, properties {soundWaveAssetPtr=&/Device_S19_RockDeadDrop_01}}
 {id=WavePlayer_2, properties {soundWaveAssetPtr=&/Device_S19_RockDeadDrop_02}}
]
properties {
 attenuationSettings=&/Device_RockDeadDrop_Atten
 maxDistance=1750
 soundClassObject=&Fort_Quests
 volumeMultiplier=1
}
